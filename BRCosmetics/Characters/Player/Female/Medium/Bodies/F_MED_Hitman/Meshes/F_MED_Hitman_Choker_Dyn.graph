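# Physics asset for the choker on the F_MED_Hitman female character, giving its dynamic necklace bones collision shapes and joints.
# A box on dyn_necklace_AIM and a small capsule on dyn_necklace_MSTR form the necklace, while capsules on neck_01 and spine_05 approximate the neck and upper torso.
# Constraints chain AIM to MSTR to spine_05, plus neck_01 to spine_05, all with free swing and twist, leaving the set angular limits inactive.
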PhysicsAsset /BRCosmetics/Characters/Player/Female/Medium/Bodies/F_MED_Hitman/Meshes/F_MED_Hitman_Choker_Dyn
bodies {
 dyn_necklace_AIM [{shapeType=Box, extentX=2.25, extentY=5, extentZ=1.3, center=<1 0 0>, rotation=<rot 7.8045 0 0>}]
 dyn_necklace_MSTR [{shapeType=Capsule, length=0.5, radius=0.5}]
 neck_01 [{shapeType=Capsule, length=3.2408, radius=3.9738, center=<2.3002 -0.068 0.0001>, rotation=<rot 90 0 4.3182>}]
 spine_05 [{shapeType=Capsule, length=14.2469, radius=6.2132, center=<0.4829 -0.2689 0.0215>, rotation=<rot 0 66.2943 -89.5829>}]
}
constraints [
 {
  bone1Name=dyn_necklace_AIM
  bone2Name=dyn_necklace_MSTR
  swing1LimitDegrees=45
  swing1Motion=Free
  swing2LimitDegrees=45
  swing2Motion=Free
  twistLimitDegrees=45
  twistMotion=Free
 }
 {
  bone1Name=neck_01
  bone2Name=spine_05
  swing1LimitDegrees=45
  swing1Motion=Free
  swing2LimitDegrees=45
  swing2Motion=Free
  twistLimitDegrees=45
  twistMotion=Free
 }
 {
  bone1Name=dyn_necklace_MSTR
  bone2Name=spine_05
  swing1LimitDegrees=45
  swing1Motion=Free
  swing2LimitDegrees=45
  swing2Motion=Free
  twistLimitDegrees=45
  twistMotion=Free
 }
]
