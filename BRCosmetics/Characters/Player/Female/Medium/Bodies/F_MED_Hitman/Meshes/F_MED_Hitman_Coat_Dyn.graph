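# Physics asset for the coat of the medium female Hitman player cosmetic, giving its dyn_ coat, left collar and shirt bones physics bodies.
# Back, side and front coat flaps and the shirt are each constrained to a tiny MSTR anchor capsule with limited swing and twist.
# The left collar is a two-link chain that bends on one swing axis only.
# Spine and left clavicle capsules supply unconstrained body shapes.
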PhysicsAsset /BRCosmetics/Characters/Player/Female/Medium/Bodies/F_MED_Hitman/Meshes/F_MED_Hitman_Coat_Dyn
bodies {
 clavicle_l [{shapeType=Capsule, length=6.5019, radius=4.0543, center=<8.1334 -1.2075 -0.2569>, rotation=<rot 84.9998 0 0>}]
 dyn_coat_MSTR [{shapeType=Capsule, length=0.5, radius=0.5, rotation=<rot 0 -90 0>}]
 dyn_coat_bk [{shapeType=Capsule, length=11.8841, radius=3.5828, center=<4.1572 0.0692 2.6359>, rotation=<rot 88.4154 92.2943 1.7518>}]
 dyn_coat_bk_MSTR [{shapeType=Capsule, length=0.5, radius=0.5, rotation=<rot 0 -90 0>}]
 dyn_coat_ft_l [{shapeType=Capsule, length=2, radius=1.5, center=<2 0 0>, rotation=<rot 90 0 2>}]
 dyn_coat_ft_l_MSTR [{shapeType=Capsule, length=0.5, radius=0.5, rotation=<rot 0 -90 0>}]
 dyn_coat_ft_r [{shapeType=Capsule, length=2, radius=1.5, center=<2 0 0>, rotation=<rot 90 0 2>}]
 dyn_coat_ft_r_MSTR [{shapeType=Capsule, length=0.5, radius=0.5, rotation=<rot 0 -90 0>}]
 dyn_coat_l [{shapeType=Capsule, length=2, radius=1, center=<2 0 0>, rotation=<rot 90 0 2>}]
 dyn_coat_l_MSTR [{shapeType=Capsule, length=0.5, radius=0.5, rotation=<rot 0 -90 0>}]
 dyn_coat_r [{shapeType=Capsule, length=2, radius=1, center=<2 0 0>, rotation=<rot 90 0 2>}]
 dyn_coat_r_MSTR [{shapeType=Capsule, length=0.5, radius=0.5, rotation=<rot 0 -90 0>}]
 dyn_collar_01_l [{shapeType=Capsule, length=3, radius=1.2, center=<1.5 0 0>, rotation=<rot 0 0 90>}]
 dyn_collar_02_l [{shapeType=Box, extentX=2.985, extentY=5, extentZ=1.2, center=<3.2434 -0.3137 0>, rotation=<rot 0 -30.0001 0>}]
 dyn_collar_l_MSTR [{shapeType=Capsule, length=0.5, radius=0.5, rotation=<rot 0 -90 0>}]
 dyn_shirt [{shapeType=Capsule, length=3.2, radius=1, center=<3.5 0 0>, rotation=<rot -5 20 -75>}]
 dyn_shirt_MSTR [{shapeType=Capsule, length=0.5, radius=0.5, rotation=<rot 0 -90 0>}]
 spine_03 [{shapeType=Capsule, length=3.8624, radius=8.8287, center=<1.0788 2.0531 0>}]
 spine_04 [{shapeType=Capsule, length=4.9302, radius=8.0503, center=<1.4093 2.8875 0.0001>}]
}
constraints [
 {
  bone1Name=dyn_collar_01_l
  bone2Name=dyn_collar_l_MSTR
  swing1LimitDegrees=45
  swing1Motion=Locked
  swing2LimitDegrees=25
  twistLimitDegrees=45
  twistMotion=Locked
 }
 {
  bone1Name=dyn_collar_02_l
  bone2Name=dyn_collar_01_l
  swing1LimitDegrees=45
  swing1Motion=Locked
  swing2LimitDegrees=25
  twistLimitDegrees=45
  twistMotion=Locked
 }
 {bone1Name=dyn_coat_bk, bone2Name=dyn_coat_bk_MSTR, swing1LimitDegrees=10, swing2LimitDegrees=30, twistLimitDegrees=5}
 {bone1Name=dyn_coat_l, bone2Name=dyn_coat_l_MSTR, swing1LimitDegrees=10, swing2LimitDegrees=30, twistLimitDegrees=5}
 {bone1Name=dyn_coat_ft_l, bone2Name=dyn_coat_ft_l_MSTR, swing1LimitDegrees=10, swing2LimitDegrees=30, twistLimitDegrees=5}
 {bone1Name=dyn_coat_r, bone2Name=dyn_coat_r_MSTR, swing1LimitDegrees=10, swing2LimitDegrees=30, twistLimitDegrees=5}
 {bone1Name=dyn_coat_ft_r, bone2Name=dyn_coat_ft_r_MSTR, swing1LimitDegrees=10, swing2LimitDegrees=30, twistLimitDegrees=5}
 {bone1Name=dyn_shirt, bone2Name=dyn_shirt_MSTR, swing1LimitDegrees=5, swing2LimitDegrees=45, twistLimitDegrees=5}
]
