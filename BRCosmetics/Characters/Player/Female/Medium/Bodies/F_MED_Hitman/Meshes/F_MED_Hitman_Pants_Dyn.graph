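# Physics asset for the dynamic trouser legs on the pants mesh of the female medium Hitman player cosmetic.
# Each side pairs a long capsule body on its trouser leg bone with a tiny capsule on the matching MSTR bone.
# A tightly limited swing and twist constraint joins each pair, restricting the leg to slight movement relative to its MSTR body.
# The left and right sides mirror each other's setup, with no constraint between them.
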PhysicsAsset /BRCosmetics/Characters/Player/Female/Medium/Bodies/F_MED_Hitman/Meshes/F_MED_Hitman_Pants_Dyn
bodies {
 dyn_trouser_leg_l [{shapeType=Capsule, length=20, radius=4.5, center=<10.411 0 0.8029>, rotation=<rot 85.0001 0 0>}]
 dyn_trouser_leg_l_MSTR [{shapeType=Capsule, length=1, radius=1}]
 dyn_trouser_leg_r [{shapeType=Capsule, length=20, radius=4.5, center=<10.411 0 0.8029>, rotation=<rot 85.0001 0 0>}]
 dyn_trouser_leg_r_MSTR [{shapeType=Capsule, length=1, radius=1}]
}
constraints [
 {bone1Name=dyn_trouser_leg_l, bone2Name=dyn_trouser_leg_l_MSTR, swing1LimitDegrees=3, swing2LimitDegrees=3, twistLimitDegrees=0.5}
 {bone1Name=dyn_trouser_leg_r, bone2Name=dyn_trouser_leg_r_MSTR, swing1LimitDegrees=3, swing2LimitDegrees=3, twistLimitDegrees=0.5}
]
