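# Animation Blueprint for the female Hitman body cosmetic, layering secondary motion for coat, pants, choker and necklace onto the pose copied from the attached parent mesh.
# Thigh and neck rotations, remapped each update, drive corrective glute and necklace offsets; head values are computed but unused.
# A weighted spine constraint anchors the coat before three RigidBody simulations.
# Skydiving flags then add coat bone rotations, with a slow blend for floating.
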AnimBlueprint /BRCosmetics/Characters/Player/Female/Medium/Bodies/F_MED_Hitman/Meshes/F_MED_Hitman_AnimBP
target_skeleton=&/F_MED_Hitman_Skeleton
graphs [
 AnimGraph {nodes [
   n1 = "Transform (Modify) Bone - Bone: dyn_glute_l"(
    ComponentPose=Local To Component(LocalPose=Copy Pose From Mesh() {bUseAttachedParent=true})
    Alpha=Get thigh_rot_up_l() -> (thigh_rot_up_l:double)
    Translation="-12,-6,-1"
    Rotation="0,0,0"
    Scale="1,1,1"
   ) {
    translationMode=BMM_Additive
    translationSpace=BCS_BoneSpace
    alphaBoolBlend {blendOption=Linear}
    boneToModify {boneName=dyn_glute_l}
    translation=<-12 -6 -1>
   }
   n2 = "Transform (Modify) Bone - Bone: dyn_glute_r"(ComponentPose=@n1, Alpha=Get thigh_rot_up_r() -> (thigh_rot_up_r:double), Translation="-12,-6,-1", Rotation="0,0,0", Scale="1,1,1") {
    translationMode=BMM_Additive
    translationSpace=BCS_BoneSpace
    alphaBoolBlend {blendOption=Linear}
    boneToModify {boneName=dyn_glute_r}
    translation=<-12 -6 -1>
   }
   n3 = "Transform (Modify) Bone - Bone: dyn_necklace_MSTR"(ComponentPose=@n2, Alpha=Get neck01_rot_dn() -> (neck01_rot_dn:double), Translation="0,0,1", Rotation="0,0,0", Scale="1,1,1") {
    translationMode=BMM_Additive
    translationSpace=BCS_BoneSpace
    alphaBoolBlend {blendOption=Linear}
    boneToModify {boneName=dyn_necklace_MSTR}
    translation {z=1}
   }
   n4 = "Constraint - dyn_coat_MSTR (PPP)"!(ComponentPose=@n3, Alpha=1, ConstraintWeights_0=0.25, ConstraintWeights_1:float=0.4, ConstraintWeights_2:float=0.35) -> (Pose) {
    alphaBoolBlend {blendOption=Linear}
    boneToModify {boneName=dyn_coat_MSTR}
    constraintSetup [{targetBone {boneName=spine_04}}, {targetBone {boneName=spine_05}}, {targetBone {boneName=spine_03}}]
    constraintWeights [1,1,1]
   }
   n5 = RigidBody(ComponentPose=@n4) {
    bClampLinearTranslationLimitToRefPose=true
    bOverrideWorldGravity=true
    overridePhysicsAsset=&/F_MED_Hitman_Coat_Dyn
    simulationSpace=BaseBoneSpace
    alphaBoolBlend {blendOption=Linear}
    baseBoneRef {boneName=spine_05}
    componentAppliedLinearAccClamp=<2000 2000 2000>
    componentLinearAccScale=<0.3 0.3 0.3>
    componentLinearVelScale=<0.3 0.3 0.3>
    overrideWorldGravity {z=-980}
   }
   n6 = RigidBody(ComponentPose=@n5, Alpha=0.5) {
    bClampLinearTranslationLimitToRefPose=true
    bOverrideWorldGravity=true
    overridePhysicsAsset=&/F_MED_Hitman_Pants_Dyn
    simulationSpace=BaseBoneSpace
    alphaBoolBlend {blendOption=Linear}
    baseBoneRef {boneName=root}
    componentAppliedLinearAccClamp=<1500 1500 1500>
    componentLinearAccScale=<0.2 0.2 0.2>
    componentLinearVelScale=<0.2 0.2 0.2>
    overrideWorldGravity {z=-980}
   }
   n7 = RigidBody(ComponentPose=@n6) {
    bClampLinearTranslationLimitToRefPose=true
    bOverrideWorldGravity=true
    overridePhysicsAsset=&/F_MED_Hitman_Choker_Dyn
    simulationSpace=BaseBoneSpace
    alphaBoolBlend {blendOption=Linear}
    baseBoneRef {boneName=spine_05}
    componentAppliedLinearAccClamp=<0 0 0>
   }
   n8 = Get bIsSkyDiving()
   n9 = "Transform (Modify) Bone - Bone: dyn_coat_bk"!(ComponentPose=@n7, bAlphaBoolEnabled:bool=@n8, Translation="0,0,0", Rotation="5.990,0.140,-0.618", Scale="1,1,1") -> (Pose) {
    alphaInputType=Bool
    rotationMode=BMM_Additive
    rotationSpace=BCS_BoneSpace
    boneToModify {boneName=dyn_coat_bk}
    rotation=<rot 5.99 0.14 -0.618>
   }
   n10 = "Transform (Modify) Bone - Bone: dyn_coat_l"!(ComponentPose=@n9, bAlphaBoolEnabled:bool=@n8, Translation="0,0,0", Rotation="7.390,-10.862,-0.054", Scale="1,1,1") -> (Pose) {
    alphaInputType=Bool
    rotationMode=BMM_Additive
    rotationSpace=BCS_BoneSpace
    boneToModify {boneName=dyn_coat_l}
    rotation=<rot 7.39 -10.862 -0.054>
   }
   n11 = "Transform (Modify) Bone - Bone: dyn_coat_bk"(ComponentPose=@n10, bAlphaBoolEnabled=Get bIsSkydiveFloating(), Translation="0,0,0", Rotation="25,0,0", Scale="1,1,1") {
    alphaInputType=Bool
    rotationMode=BMM_Additive
    rotationSpace=BCS_BoneSpace
    alphaBoolBlend {blendInTime=5, blendOutTime=5}
    boneToModify {boneName=dyn_coat_bk}
    rotation {pitch=25}
   }
   Output Pose(Result=Component To Local(ComponentPose=@n11)) {layerGroup=None}
  ]}
 EventGraph {nodes [
   Event BlueprintUpdateAnimation()
   n1 = BreakTransform(InTransform=GetDeltaTransformFromRefPose(self=GetOwnerSkeletalMeshComponent(), BoneName=thigh_r, BaseName=pelvis))
   > Set thigh_rot_up_r(thigh_rot_up_r:double=MapRangeClamped(Value=BreakRotator(InRot=@n1.Rotation).Yaw, InRangeA=0, InRangeB=-100)) -> (Output_Get:double)
   n2 = BreakTransform(InTransform=GetDeltaTransformFromRefPose(self=GetOwnerSkeletalMeshComponent(), BoneName=thigh_l, BaseName=pelvis))
   > Set thigh_rot_up_l(thigh_rot_up_l:double=MapRangeClamped(Value=BreakRotator(InRot=@n2.Rotation).Yaw, InRangeA=-30, InRangeB=-100)) -> (Output_Get:double)
   n3 = GetOwnerSkeletalMeshComponent()
   n4 = BreakRotator(InRot=BreakTransform(InTransform=GetDeltaTransformFromRefPose(self=@n3, BoneName=head, BaseName=neck_01)).Rotation)
   > Set head_rot_up(head_rot_up:double=MapRangeClamped(Value=@n4.Yaw, InRangeA=0, InRangeB=-50)) -> (Output_Get:double)
   > Set head_rot_dn(head_rot_dn:double=MapRangeClamped(Value=@n4.Yaw, InRangeA=0, InRangeB=15)) -> (Output_Get:double)
   n5 = BreakTransform(InTransform=GetDeltaTransformFromRefPose(self=GetOwnerSkeletalMeshComponent(), BoneName=neck_01, BaseName=spine_05))
   > Set neck01_rot_dn(neck01_rot_dn:double=MapRangeClamped(Value=BreakRotator(InRot=@n5.Rotation).Yaw, InRangeA=0, InRangeB=16)) -> (Output_Get:double)
  ]}
]
variables [
 head_rot_dn {type=number}
 head_rot_up {type=number}
 neck01_rot_dn {type=number}
 thigh_rot_up_l {type=number}
 thigh_rot_up_r {type=number}
]
