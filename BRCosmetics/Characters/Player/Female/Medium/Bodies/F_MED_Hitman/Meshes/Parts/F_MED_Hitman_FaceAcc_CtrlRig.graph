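# Control Rig for the female Hitman face accessory that shifts the eye patch bones as the face underneath moves.
# The forward solve reads the Y position, relative to faceAttach, of the left inner cheek, left mid brow and center brow bones.
# Each value passes through a clamped remap to weight an additive local translation on the lower patch, upper patch and strap bones respectively, acting as a procedural corrective.
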ControlRigBlueprint /BRCosmetics/Characters/Player/Female/Medium/Bodies/F_MED_Hitman/Meshes/Parts/F_MED_Hitman_FaceAcc_CtrlRig
graphs [
 RigVMFunctionLibrary {links [], nodes []}
 RigVMModel {nodes [
   n1 = &RigUnit_BeginExecution()
   /Script/ControlRig.RigUnit_ModifyBoneTransforms(
    BoneToModify=((Bone="dyn_eyePatch_lw",Transform=(Rotation=(X=0,Y=0,Z=0,W=1),Translation=(X=0,Y=1.8,Z=0.1),Scale3D=(X=1,Y=1,Z=1))))
    Mode=AdditiveLocal
    WeightMaximum=1
    WeightMinimum=0
    ExecutePin=@n1.ExecutePin
    Weight=&RigVMFunction_MathFloatRemap(
     SourceMaximum=9.5
     SourceMinimum=8
     TargetMinimum=0
     bClamp=true
     Value=&RigUnit_GetRelativeBoneTransform(Bone=L_cheek_inner, Space=faceAttach).Transform.Translation.Y
    ).Result
    BoneToModify.0.Transform.Rotation=/Script/RigVM.RigVMFunction_MathQuaternionFromRotator(Rotator=(Pitch=0,Yaw=0,Roll=0)) {is=struct RigVMUnitNode}.Result
   ) {is=struct RigVMUnitNode}
   > &RigUnit_ModifyBoneTransforms(
    BoneToModify=((Bone="dyn_eyePatch_up",Transform=(Rotation=(X=0,Y=0,Z=0,W=1),Translation=(X=0,Y=0.3,Z=-0.1),Scale3D=(X=1,Y=1,Z=1))))
    Mode=AdditiveLocal
    WeightMaximum=1
    WeightMinimum=0
    Weight=&RigVMFunction_MathFloatRemap(
     SourceMaximum=9.8
     SourceMinimum=9.1
     TargetMinimum=0
     bClamp=true
     Value=&RigUnit_GetRelativeBoneTransform(Bone=L_brow_mid, Space=faceAttach).Transform.Translation.Y
    ).Result
    BoneToModify.0.Transform.Rotation=&RigVMFunction_MathQuaternionFromRotator(Rotator=(Pitch=0,Yaw=0,Roll=0)).Result
   )
   > &RigUnit_ModifyBoneTransforms(
    BoneToModify=((Bone="dyn_eyePatch_strap",Transform=(Rotation=(X=0,Y=0,Z=0,W=1),Translation=(X=0,Y=2,Z=0),Scale3D=(X=1,Y=1,Z=1))))
    Mode=AdditiveLocal
    WeightMaximum=1
    WeightMinimum=0
    Weight=&RigVMFunction_MathFloatRemap(
     SourceMaximum=11
     SourceMinimum=10
     TargetMinimum=0
     bClamp=true
     Value=&RigUnit_GetRelativeBoneTransform(Bone=C_brow_mid, Space=faceAttach).Transform.Translation.Y
    ).Result
    BoneToModify.0.Transform.Rotation=&RigVMFunction_MathQuaternionFromRotator(Rotator=(Pitch=0,Yaw=0,Roll=0)).Result
   )
  ]}
]
hierarchy [
 root {loc=<0 0 0>, rot=<0 0 0 1>}
 pelvis {parent=root, loc=<0 -0.5301 98.0571>, rot=<0.0127 0.707 -0.0127 -0.707>}
 spine_01 {parent=pelvis, loc=<6.1208 -0.9328 0>, rot=<0 0 0.034 -0.9994>}
 spine_02 {parent=spine_01, loc=<6.5996 -0.2095 0>, rot=<0 0 0.0238 -0.9997>}
 spine_03 {parent=spine_02, loc=<6.4914 -0.1237 0>, rot=<0 0 0.0149 -0.9999>}
 spine_04 {parent=spine_03, loc=<6.9598 0.0794 0>, rot=<0 0 -0.0132 -0.9999>}
 spine_05 {parent=spine_04, loc=<6.9987 0 0>, rot=<0 0 -0.0015 -1>}
 clavicle_l {parent=spine_05, loc=<5.4831 -0.1053 -3.2946>, rot=<0.0465 -0.7289 -0.009 -0.6829>}
 upperarm_l {parent=clavicle_l, loc=<9.8658 -1.2136 -0.4105>, rot=<0.0164 -0.3245 -0.0269 -0.9454>}
 upperarm_twist_01_l {parent=upperarm_l, loc=<7.2135 -0.1134 0>, rot=<0 0 0 -1>}
 upperarm_twist_02_l {parent=upperarm_l, loc=<19.7998 -0.1134 0>, rot=<0 0 0 -1>}
 clavicle_r {parent=spine_05, loc=<5.4831 -0.1053 3.2946>, rot=<0.7289 0.0465 -0.6829 0.009>}
 upperarm_r {parent=clavicle_r, loc=<-9.8658 1.2136 0.4104>, rot=<0.0164 -0.3245 -0.0269 -0.9454>}
 upperarm_twist_01_r {parent=upperarm_r, loc=<-7.2135 0.1134 0>, rot=<0 0 0 1>}
 upperarm_twist_02_r {parent=upperarm_r, loc=<-19.7998 0.1134 0>, rot=<0 0 0 1>}
 neck_01 {parent=spine_05, loc=<11.3138 0.1529 0>, rot=<0 0 -0.2003 -0.9797>}
 neck_02 {parent=neck_01, loc=<4.5627 0 0>, rot=<0 0 0.0504 -0.9987>}
 head {parent=neck_02, loc=<5.3914 0 0>, rot=<0 0 0.115 -0.9934>}
 dyn_earrings_l {parent=head, loc=<-0.1953 0.2023 -5.7484>, rot=<0.0039 -0.9397 -0.342 0.0014>}
 dyn_earrings_r {parent=head, loc=<-0.195 0.2167 5.7372>, rot=<-0.0014 0.3508 0.9364 -0.0039>}
 dyn_frings_a_01 {parent=head, loc=<9.9226 8.3269 0.5407>, rot=<0.2283 0.6363 0.7368 0.0043>}
 dyn_frings_a_02 {parent=dyn_frings_a_01, loc=<2.2175 -0.02 -0.1177>, rot=<0 -0.1544 0 -0.988>}
 dyn_frings_b_01 {parent=head, loc=<9.3896 8.3887 -2.6343>, rot=<-0.196 -0.714 -0.6658 -0.0923>}
 dyn_frings_b_02 {parent=dyn_frings_b_01, loc=<2.5478 0.0275 -0.2678>, rot=<-0.0239 -0.2221 0.009 -0.9747>}
 dyn_hair_l_MSTR {parent=head, loc=<3.1864 -2.3075 -7.7035>, rot=<0.0532 -0.9906 0.0843 0.0939>}
 dyn_hair_l_01 {parent=dyn_hair_l_MSTR, loc=<0 0 0>, rot=<0 0 0 -1>}
 dyn_hair_l_02 {parent=dyn_hair_l_01, loc=<6.075 -0.2814 0.197>, rot=<-0.0321 0.2014 0.0701 -0.9765>}
 dyn_hair_l_03 {parent=dyn_hair_l_02, loc=<4.8069 -0.1608 0.1816>, rot=<0.0378 -0.0107 0.2088 -0.9772>}
 dyn_hair_bk_MSTR {parent=head, loc=<8.4097 -6.9776 -0.0004>, rot=<0.0629 -0.7043 0.7043 -0.0629>}
 dyn_hair_bk_01 {parent=dyn_hair_bk_MSTR, loc=<0 0 0>, rot=<0 0 0 -1>}
 dyn_hair_bk_02 {parent=dyn_hair_bk_01, loc=<4.9973 0 0>, rot=<0 -0.0294 0 -0.9996>}
 dyn_hair_bk_03 {parent=dyn_hair_bk_02, loc=<6.5123 0 0>, rot=<0 0.1497 0 -0.9887>}
 dyn_hair_bk_04 {parent=dyn_hair_bk_03, loc=<5.0129 0 0.0855>, rot=<0 -0.2042 0 -0.9789>}
 dyn_hair_bk_05 {parent=dyn_hair_bk_04, loc=<4.0518 0 -0.0974>, rot=<0 0 0 -1>}
 dyn_hair_r_bk_MSTR {parent=head, loc=<7.4168 -4.8201 6.6629>, rot=<0.0282 -0.3374 0.9379 -0.0752>}
 dyn_hair_r_bk_01 {parent=dyn_hair_r_bk_MSTR, loc=<0 0 0>, rot=<0 0 0 1>}
 dyn_hair_r_bk_02 {parent=dyn_hair_r_bk_01, loc=<5.4036 0 0>, rot=<-0.0205 -0.0924 -0.0866 -0.9917>}
 dyn_hair_r_bk_03 {parent=dyn_hair_r_bk_02, loc=<5.2959 0 0>, rot=<0.0826 0.3293 0.2039 -0.9183>}
 dyn_hair_a_r_01 {parent=dyn_hair_r_bk_03, loc=<2.2977 1.1757 -1.1202>, rot=<0.1657 0.0459 -0.0049 -0.9851>}
 dyn_hair_a_r_02 {parent=dyn_hair_a_r_01, loc=<3.5849 -0.0314 -0.0462>, rot=<0.0483 -0.1808 -0.2536 -0.949>}
 dyn_hair_b_r_01 {parent=dyn_hair_r_bk_03, loc=<1.2529 -1.6418 -0.5066>, rot=<-0.2254 0.0309 0.1782 -0.9573>}
 dyn_hair_b_r_02 {parent=dyn_hair_b_r_01, loc=<4.2294 0.4196 0.3391>, rot=<-0.0888 -0.2008 -0.2198 -0.9505>}
 dyn_hair_r_MSTR {parent=head, loc=<3.4438 -2.6461 7.3292>, rot=<0.0508 -0.0928 0.9923 -0.064>}
 dyn_hair_r_01 {parent=dyn_hair_r_MSTR, loc=<0 0 0>, rot=<0 0 0 1>}
 dyn_hair_r_02 {parent=dyn_hair_r_01, loc=<3.4336 0.0688 -0.376>, rot=<0.0524 -0.1718 0.1123 -0.9773>}
 dyn_hair_r_03 {parent=dyn_hair_r_02, loc=<3.574 0.1934 0.0371>, rot=<-0.1872 0.2195 0.0783 -0.9543>}
 dyn_hair_r_04 {parent=dyn_hair_r_03, loc=<3.8207 -0.0131 -0.0209>, rot=<0.0421 -0.1321 0.0753 -0.9875>}
 faceAttach {parent=head, loc=<-3.5269 -0.4277 -0.0076>, rot=<0.0082 -0.7071 -0.0082 -0.7071>, scale=<0.89 0.89 0.89>}
 R_brow_mid {parent=faceAttach, loc=<-3.5601 9.1286 9.716>, rot=<0 0.0034 -0.2553 -0.9669>}
 L_eye {parent=faceAttach, loc=<3.2456 6.9699 8.2141>, rot=<0 0 0.9998 0.0187>}
 L_eye_lid_upper_mid {parent=L_eye, loc=<0.0251 -1.4793 0.2702>, rot=<0.0517 0 0 -0.9987>}
 L_eye_lid_lower_mid {parent=L_eye, loc=<0.0223 -1.2727 -0.8659>, rot=<-0.0225 0 0 -0.9997>}
 C_brow_mid {parent=faceAttach, loc=<-0.0086 9.9856 8.8402>, rot=<-0.0074 0 0 -1>}
 L_brow_mid {parent=faceAttach, loc=<3.5429 9.1286 9.716>, rot=<0 -0.0034 0.2553 -0.9669>}
 L_cheek_inner {parent=faceAttach, loc=<4.3097 8.0265 3.3227>, rot=<-0.0105 -0.2271 0.4858 -0.844>}
 dyn_eyePatch_strap {parent=head, loc=<4.9479 8.8668 0.01>, rot=<0.0029 -0.7071 -0.0029 -0.7071>}
 dyn_eyePatch_up {parent=head, loc=<4.951 8.3435 -3.3563>, rot=<0.0029 -0.7071 -0.0029 -0.7071>}
 dyn_eyePatch_lw {parent=head, loc=<1.7696 7.287 -3.3762>, rot=<0.0029 -0.7071 -0.0029 -0.7071>}
]
