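# Physics asset for the Hitman female medium character's face-accessory/hair part, setting up secondary motion for hair, fringe and earrings.
# Limited swing/twist constraints chain the dynamic capsules from the head, with left, back and right hair chains rooted on small MSTR bodies.
# Two strands branch from the right-back chain, one cross-linked to the back chain.
# Head, neck, upper spine, clavicle and right upper-arm capsules form the body volume, joined by free-angular constraints.
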PhysicsAsset /BRCosmetics/Characters/Player/Female/Medium/Bodies/F_MED_Hitman/Meshes/Parts/F_MED_Hitman_FaceAcc_Hair_Dyn
bodies {
 clavicle_l [{shapeType=Capsule, length=9.0933, radius=5.4923, center=<4.8479 -0.7604 -0.2424>, rotation=<rot 90 0 0>}]
 clavicle_r [{shapeType=Capsule, length=9.2666, radius=5.2879, center=<-4.1636 1.6292 -0.1031>, rotation=<rot 84.9995 0 0>}]
 dyn_earrings_l [{shapeType=Capsule, length=1.1345, radius=1, center=<1.5 0 0>, rotation=<rot 90 0 0>}]
 dyn_earrings_r [{shapeType=Capsule, length=1.1345, radius=1, center=<1.5 0 0>, rotation=<rot 90 0 0>}]
 dyn_frings_a_01 [{shapeType=Capsule, length=0.4585, radius=0.9124, center=<1 0 0>, rotation=<rot 90 0 90>}]
 dyn_frings_a_02 [{shapeType=Capsule, length=0.574, radius=0.9342, center=<1 0 0>, rotation=<rot 90 0 90>}]
 dyn_frings_b_01 [{shapeType=Capsule, length=2, radius=1, center=<1 0 0>, rotation=<rot 0 0 -103.5296>}]
 dyn_frings_b_02 [{shapeType=Capsule, length=2.25, radius=1, center=<1 0 0>, rotation=<rot 0 0.0001 -99.9647>}]
 dyn_hair_a_r_01 [{shapeType=Capsule, length=1.426, radius=1.574, center=<3 0 0>, rotation=<rot 90 0 0>}]
 dyn_hair_a_r_02 [{shapeType=Capsule, length=1.1835, radius=1.3862, center=<2 0 0>, rotation=<rot 90 18.435 -8.3068>}]
 dyn_hair_b_r_01 [{shapeType=Capsule, length=1.3589, radius=1.6411, center=<2.9729 0 0>, rotation=<rot 90 0 -10.0001>}]
 dyn_hair_b_r_02 [{shapeType=Capsule, length=1.9248, radius=1.1449, center=<1.8114 0 0>, rotation=<rot 57.806 36.4233 0>}]
 dyn_hair_bk_01 [{shapeType=Capsule, length=5, radius=2.75, center=<2.3207 0 0>, rotation=<rot 0 0 90.0201>}]
 dyn_hair_bk_02 [{shapeType=Capsule, length=7, radius=2.75, center=<3.2707 0 0>, rotation=<rot 0 0 89.9373>}]
 dyn_hair_bk_03 [{shapeType=Capsule, length=5.75, radius=2.75, center=<3.3431 0 0>, rotation=<rot 0.0001 0 90.2117>}]
 dyn_hair_bk_04 [{shapeType=Capsule, length=7, radius=2, center=<2.3987 0 0>, rotation=<rot 0 0 90.1839>}]
 dyn_hair_bk_05 [{shapeType=Capsule, length=6, radius=2.5, center=<4.2095 0 0>, rotation=<rot 0 0 90.1265>}]
 dyn_hair_bk_MSTR [{shapeType=Capsule, length=0.5, radius=0.5}]
 dyn_hair_l_01 [{shapeType=Capsule, length=1.75, radius=2, center=<2.4273 0 0>, rotation=<rot 0 0 90.0453>}]
 dyn_hair_l_02 [{shapeType=Capsule, length=2.9989, radius=2.5011, center=<2 0 0>, rotation=<rot 90 0 0>}]
 dyn_hair_l_03 [{shapeType=Capsule, length=1.5018, radius=2.2774, center=<1.5862 0 0>, rotation=<rot 90 -36.8699 -11.8662>}]
 dyn_hair_l_MSTR [{shapeType=Capsule, length=0.5, radius=0.5}]
 dyn_hair_r_01 [{shapeType=Capsule, length=3, radius=1.4, center=<2 0 0>, rotation=<rot 90 0 0>}]
 dyn_hair_r_02 [{shapeType=Capsule, length=3, radius=1.5, center=<2 0 0>, rotation=<rot 90 0 0>}]
 dyn_hair_r_03 [{shapeType=Capsule, length=1, radius=2, center=<3.2 0 0>, rotation=<rot 80 0 0>}]
 dyn_hair_r_04 [{shapeType=Capsule, length=2.5, radius=1.5, center=<2 0 0>, rotation=<rot 79.9997 0 0>}]
 dyn_hair_r_MSTR [{shapeType=Capsule, length=0.5, radius=0.5}]
 dyn_hair_r_bk_01 [{shapeType=Capsule, length=4, radius=1.4, center=<3 0 0>, rotation=<rot 90 0 0>}]
 dyn_hair_r_bk_02 [{shapeType=Capsule, length=3.75, radius=2.75, center=<2.7887 0 0>}]
 dyn_hair_r_bk_MSTR [{shapeType=Capsule, length=0.5, radius=0.5}]
 head [{shapeType=Capsule, length=2.4642, radius=7.2894, center=<4.4963 0.9916 0>, rotation=<rot 90 0 -10>}]
 neck_01 [
  {shapeType=Capsule, length=3.1261, radius=4.1303, center=<4.3902 1.1208 0>, rotation=<rot 90 0 0>}
  {shapeType=Capsule, length=6.492, radius=4.1303, center=<3.3264 -1.8764 -0.0001>, rotation=<rot 90 0 0>}
 ]
 spine_05 [
  {shapeType=Capsule, length=21.155, radius=5.0215, center=<5.8075 -1.7956 -0.0001>, rotation=<rot 0 -90 0>}
  {shapeType=Capsule, length=7.5491, radius=5.0215, center=<5.0541 2.463 0>, rotation=<rot 0 -90 0>}
  {shapeType=Capsule, length=14.487, radius=6.8851, center=<0.5271 -0.8719 -0.0001>, rotation=<rot 0 -90 0>}
  {shapeType=Capsule, length=5.1958, radius=4.0353, center=<9.0545 3.8921 -5.4028>, rotation=<rot -55.6913 -82.5184 -6.526>}
 ]
 upperarm_r [{shapeType=Capsule, length=13.9973, radius=4.628, center=<-8.6872 0.1817 0.0824>, rotation=<rot 90 0 0>}]
}
constraints [
 {bone1Name=dyn_earrings_l, bone2Name=head, swing1LimitDegrees=30, swing2LimitDegrees=30, twistLimitDegrees=5}
 {bone1Name=dyn_earrings_r, bone2Name=head, swing1LimitDegrees=30, swing2LimitDegrees=30, twistLimitDegrees=5}
 {bone1Name=dyn_frings_a_01, bone2Name=head, swing1LimitDegrees=2, swing2LimitDegrees=10, twistLimitDegrees=2}
 {bone1Name=dyn_frings_a_02, bone2Name=dyn_frings_a_01, swing1LimitDegrees=2, swing2LimitDegrees=10, twistLimitDegrees=2}
 {bone1Name=dyn_frings_b_01, bone2Name=head, swing1LimitDegrees=2, swing2LimitDegrees=10, twistLimitDegrees=1}
 {bone1Name=dyn_frings_b_02, bone2Name=dyn_frings_b_01, swing1LimitDegrees=2, swing2LimitDegrees=10, twistLimitDegrees=1}
 {bone1Name=dyn_hair_l_02, bone2Name=dyn_hair_l_01, swing1LimitDegrees=15, swing2LimitDegrees=15, twistLimitDegrees=5}
 {bone1Name=dyn_hair_l_03, bone2Name=dyn_hair_l_02, swing1LimitDegrees=15, swing2LimitDegrees=15, twistLimitDegrees=5}
 {bone1Name=dyn_hair_bk_02, bone2Name=dyn_hair_bk_01, swing1LimitDegrees=25, swing2LimitDegrees=30, twistLimitDegrees=5}
 {bone1Name=dyn_hair_bk_03, bone2Name=dyn_hair_bk_02, swing1LimitDegrees=25, swing2LimitDegrees=20, twistLimitDegrees=5}
 {bone1Name=dyn_hair_bk_04, bone2Name=dyn_hair_bk_03, swing1LimitDegrees=25, swing2LimitDegrees=20, twistLimitDegrees=5}
 {bone1Name=dyn_hair_bk_05, bone2Name=dyn_hair_bk_04, swing1LimitDegrees=25, swing2LimitDegrees=20, twistLimitDegrees=5}
 {bone1Name=dyn_hair_r_02, bone2Name=dyn_hair_r_01, swing1LimitDegrees=30, swing2LimitDegrees=30, twistLimitDegrees=5}
 {
  bone1Name=head
  bone2Name=neck_01
  swing1LimitDegrees=45
  swing1Motion=Free
  swing2LimitDegrees=45
  swing2Motion=Free
  twistLimitDegrees=45
  twistMotion=Free
 }
 {
  bone1Name=clavicle_l
  bone2Name=spine_05
  swing1LimitDegrees=45
  swing1Motion=Free
  swing2LimitDegrees=45
  swing2Motion=Free
  twistLimitDegrees=45
  twistMotion=Free
 }
 {
  bone1Name=clavicle_r
  bone2Name=spine_05
  swing1LimitDegrees=45
  swing1Motion=Free
  swing2LimitDegrees=45
  swing2Motion=Free
  twistLimitDegrees=45
  twistMotion=Free
 }
 {
  bone1Name=neck_01
  bone2Name=spine_05
  swing1LimitDegrees=45
  swing1Motion=Free
  swing2LimitDegrees=45
  swing2Motion=Free
  twistLimitDegrees=45
  twistMotion=Free
 }
 {bone1Name=dyn_hair_r_03, bone2Name=dyn_hair_r_02, swing1LimitDegrees=15, swing2LimitDegrees=15, twistLimitDegrees=5}
 {bone1Name=dyn_hair_r_04, bone2Name=dyn_hair_r_03, swing1LimitDegrees=15, swing2LimitDegrees=15, twistLimitDegrees=5}
 {bone1Name=dyn_hair_r_bk_02, bone2Name=dyn_hair_r_bk_01, swing1LimitDegrees=65, swing2LimitDegrees=45, twistLimitDegrees=5}
 {bone1Name=dyn_hair_a_r_01, bone2Name=dyn_hair_r_bk_02, swing1LimitDegrees=10, swing2LimitDegrees=10, twistLimitDegrees=5}
 {bone1Name=dyn_hair_a_r_02, bone2Name=dyn_hair_a_r_01, swing1LimitDegrees=5, swing2LimitDegrees=5, twistLimitDegrees=2}
 {bone1Name=dyn_hair_b_r_01, bone2Name=dyn_hair_r_bk_02, swing1LimitDegrees=5, swing2LimitDegrees=5, twistLimitDegrees=5}
 {bone1Name=dyn_hair_b_r_02, bone2Name=dyn_hair_b_r_01, swing1LimitDegrees=5, swing2LimitDegrees=5, twistLimitDegrees=2}
 {
  bone1Name=upperarm_r
  bone2Name=clavicle_r
  swing1LimitDegrees=45
  swing1Motion=Free
  swing2LimitDegrees=45
  swing2Motion=Free
  twistLimitDegrees=45
  twistMotion=Free
 }
 {
  bone1Name=dyn_hair_r_MSTR
  bone2Name=head
  swing1LimitDegrees=45
  swing1Motion=Free
  swing2LimitDegrees=45
  swing2Motion=Free
  twistLimitDegrees=45
  twistMotion=Free
 }
 {
  bone1Name=dyn_hair_r_bk_MSTR
  bone2Name=head
  swing1LimitDegrees=45
  swing1Motion=Free
  swing2LimitDegrees=45
  swing2Motion=Free
  twistLimitDegrees=45
  twistMotion=Free
 }
 {
  bone1Name=dyn_hair_bk_MSTR
  bone2Name=head
  swing1LimitDegrees=45
  swing1Motion=Free
  swing2LimitDegrees=45
  swing2Motion=Free
  twistLimitDegrees=45
  twistMotion=Free
 }
 {
  bone1Name=dyn_hair_l_MSTR
  bone2Name=head
  swing1LimitDegrees=45
  swing1Motion=Free
  swing2LimitDegrees=45
  swing2Motion=Free
  twistLimitDegrees=45
  twistMotion=Free
 }
 {bone1Name=dyn_hair_l_01, bone2Name=dyn_hair_l_MSTR, swing1LimitDegrees=15, swing2LimitDegrees=15, twistLimitDegrees=5}
 {bone1Name=dyn_hair_bk_01, bone2Name=dyn_hair_bk_MSTR, swing1LimitDegrees=25, swing2LimitDegrees=30, twistLimitDegrees=5}
 {bone1Name=dyn_hair_r_bk_01, bone2Name=dyn_hair_r_bk_MSTR, swing1LimitDegrees=65, swing2LimitDegrees=45, twistLimitDegrees=5}
 {bone1Name=dyn_hair_r_01, bone2Name=dyn_hair_r_MSTR, swing1LimitDegrees=45, swing2LimitDegrees=45, twistLimitDegrees=5}
 {bone1Name=dyn_hair_a_r_02, bone2Name=dyn_hair_bk_04, swing1LimitDegrees=10, swing2LimitDegrees=10, twistLimitDegrees=10}
]
