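# Physics asset for the female medium Hitman main hair mesh, providing capsule bodies for hair, fringe and earring physics.
# Head, neck, upper spine, clavicle and upper arm capsules cover the body around the hair.
# Left, right, back and right-back hair chains hang from small MSTR bodies linked to the head with free rotation; the right-back chain forks into two strands.
# Chain links limit swing and tightly restrict twist, with the head-attached fringe stiffest.
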PhysicsAsset /BRCosmetics/Characters/Player/Female/Medium/Bodies/F_MED_Hitman/Meshes/Parts/F_MED_Hitman_FaceAcc_MainHair_Physics
bodies {
 clavicle_l [{shapeType=Capsule, length=12, radius=5.7, shapeName=clavicle_l_capsule, center=<3.5 -1 -1>, rotation=<rot 80 -0.0003 -0.0003>}]
 clavicle_r [{shapeType=Capsule, length=12, radius=5.7, shapeName=clavicle_r_capsule, center=<-3.5 1 1>, rotation=<rot 80 -0.0003 -0.0004>}]
 dyn_earrings_l [{shapeType=Capsule, length=1.1345, radius=1.3, center=<1.5 0 0>, rotation=<rot 90 0 0>}]
 dyn_earrings_r [{shapeType=Capsule, length=1.1345, radius=1, center=<1.5 0 0>, rotation=<rot 90 0 0>}]
 dyn_frings_a_01 [{shapeType=Capsule, length=0.4585, radius=0.9124, center=<1 0 0>, rotation=<rot 90 0 90>}]
 dyn_frings_a_02 [{shapeType=Capsule, length=0.574, radius=0.9342, center=<1 0 0>, rotation=<rot 90 0 90>}]
 dyn_frings_b_01 [{shapeType=Capsule, length=2, radius=1, center=<1 0 0>, rotation=<rot 0 0 -103.5296>}]
 dyn_frings_b_02 [{shapeType=Capsule, length=2.25, radius=1, center=<1 0 0>, rotation=<rot 0 0.0001 -99.9647>}]
 dyn_hair_a_r_01 [{shapeType=Capsule, length=1.426, radius=1.574, center=<2 0 0>, rotation=<rot 90 0 0>}]
 dyn_hair_a_r_02 [{shapeType=Capsule, length=1.1835, radius=1.3862, center=<1.9559 -0.0255 1.0522>, rotation=<rot 80 -150 -179.9999>}]
 dyn_hair_b_r_01 [{shapeType=Capsule, length=1.3589, radius=1.6411, center=<2 0 0>, rotation=<rot 90 0 0>}]
 dyn_hair_b_r_02 [{shapeType=Capsule, length=1.9248, radius=1.1449, center=<2.1599 0 0>, rotation=<rot 80 29.9998 0.0001>}]
 dyn_hair_bk_01 [{shapeType=Capsule, length=2, radius=2.5, shapeName=dyn_hair_bk_01_capsule, center=<4 0 0>, rotation=<rot 90 0 0>}]
 dyn_hair_bk_02 [{shapeType=Capsule, length=2, radius=2.5, shapeName=dyn_hair_bk_02_capsule, center=<4 0 0>, rotation=<rot 90 0 0>}]
 dyn_hair_bk_03 [{shapeType=Capsule, length=2, radius=2.5, shapeName=dyn_hair_bk_03_capsule, center=<4 0 0>, rotation=<rot 90 0 0>}]
 dyn_hair_bk_04 [{shapeType=Capsule, length=2, radius=2.5, shapeName=dyn_hair_bk_04_capsule, center=<4 0 0.5>, rotation=<rot 90 0 0>}]
 dyn_hair_bk_05 [{shapeType=Capsule, length=2, radius=2, shapeName=dyn_hair_bk_05_capsule, center=<4 0 1>, rotation=<rot 90 0 0>}]
 dyn_hair_bk_MSTR [{shapeType=Capsule, length=0.5, radius=0.5, rotation=<rot 0 0 90>}]
 dyn_hair_l_01 [{shapeType=Capsule, length=2, radius=2, center=<3 0 0>, rotation=<rot 90 0 0>}]
 dyn_hair_l_02 [{shapeType=Capsule, length=2, radius=2, center=<3 0 0>, rotation=<rot 90 0 0>}]
 dyn_hair_l_03 [{shapeType=Capsule, length=2, radius=2, center=<3 0 0>, rotation=<rot 90 0 0>}]
 dyn_hair_l_MSTR [{shapeType=Capsule, length=0.5, radius=0.5}]
 dyn_hair_r_01 [{shapeType=Capsule, length=2, radius=1.5, shapeName=dyn_hair_r_01_capsule, center=<3 0 0>, rotation=<rot 90 0 0>}]
 dyn_hair_r_02 [{shapeType=Capsule, length=2, radius=1.5, shapeName=dyn_hair_r_02_capsule, center=<3 0 0>, rotation=<rot 90 0 0>}]
 dyn_hair_r_03 [{shapeType=Capsule, length=2, radius=1.5, shapeName=dyn_hair_r_03_capsule, center=<3 0 0>, rotation=<rot 90 0 0>}]
 dyn_hair_r_04 [{shapeType=Capsule, length=2, radius=1.5, shapeName=dyn_hair_r_04_capsule, center=<3 0 0>, rotation=<rot 90 0 0>}]
 dyn_hair_r_MSTR [{shapeType=Capsule, length=0.5, radius=0.5}]
 dyn_hair_r_bk_01 [{shapeType=Capsule, length=4, radius=1.4, center=<3 0 0>, rotation=<rot 90 0 0>}]
 dyn_hair_r_bk_02 [{shapeType=Capsule, length=2.75, radius=1.75, center=<2.7887 0 0>, rotation=<rot 90 0 0>}]
 dyn_hair_r_bk_03 [{shapeType=Capsule, length=2, radius=2, shapeName=dyn_hair_r_bk_03_capsule, center=<2 0 0>, rotation=<rot 0 0 90>}]
 dyn_hair_r_bk_MSTR [{shapeType=Capsule, length=0.5, radius=0.5}]
 head [{shapeType=Capsule, length=6, radius=6, shapeName=head_capsule, center=<3.5 3 0>, rotation=<rot 90 0 0>}]
 neck_01 [{shapeType=Capsule, length=8, radius=4.5, shapeName=neck_01_capsule, center=<3.6683 -0.2876 0>, rotation=<rot 90 0 0>}]
 spine_05 [{shapeType=Capsule, length=12, radius=9.5, shapeName=spine_05_capsule, center=<-2.5963 0.209 0>}]
 upperarm_l [{shapeType=Capsule, length=20, radius=3.5, shapeName=upperarm_l_capsule, center=<12 -0.5 0>, rotation=<rot 90 0 0>}]
 upperarm_r [{shapeType=Capsule, length=13.9973, radius=4.628, center=<-8.6872 0.1817 0.0824>, rotation=<rot 90 0 0>}]
}
constraints [
 {bone1Name=dyn_frings_b_02, bone2Name=dyn_frings_b_01, swing1LimitDegrees=2, swing2LimitDegrees=10, twistLimitDegrees=1}
 {
  bone1Name=clavicle_l
  bone2Name=spine_05
  swing1LimitDegrees=45
  swing1Motion=Free
  swing2LimitDegrees=45
  swing2Motion=Free
  twistLimitDegrees=45
  twistMotion=Free
 }
 {
  bone1Name=clavicle_r
  bone2Name=spine_05
  swing1LimitDegrees=45
  swing1Motion=Free
  swing2LimitDegrees=45
  swing2Motion=Free
  twistLimitDegrees=45
  twistMotion=Free
 }
 {
  bone1Name=neck_01
  bone2Name=spine_05
  swing1LimitDegrees=45
  swing1Motion=Free
  swing2LimitDegrees=45
  swing2Motion=Free
  twistLimitDegrees=45
  twistMotion=Free
 }
 {
  bone1Name=upperarm_r
  bone2Name=clavicle_r
  swing1LimitDegrees=45
  swing1Motion=Free
  swing2LimitDegrees=45
  swing2Motion=Free
  twistLimitDegrees=45
  twistMotion=Free
 }
 {
  bone1Name=head
  bone2Name=neck_01
  swing1LimitDegrees=45
  swing1Motion=Free
  swing2LimitDegrees=45
  swing2Motion=Free
  twistLimitDegrees=45
  twistMotion=Free
 }
 {bone1Name=dyn_earrings_l, bone2Name=head, swing1LimitDegrees=30, swing2LimitDegrees=30, twistLimitDegrees=5}
 {bone1Name=dyn_earrings_r, bone2Name=head, swing1LimitDegrees=30, swing2LimitDegrees=30, twistLimitDegrees=5}
 {bone1Name=dyn_frings_a_01, bone2Name=head, swing1LimitDegrees=2, swing2LimitDegrees=10, twistLimitDegrees=2}
 {bone1Name=dyn_frings_b_01, bone2Name=head, swing1LimitDegrees=2, swing2LimitDegrees=10, twistLimitDegrees=1}
 {bone1Name=dyn_frings_a_02, bone2Name=dyn_frings_a_01, swing1LimitDegrees=2, swing2LimitDegrees=10, twistLimitDegrees=2}
 {bone1Name=dyn_hair_l_03, bone2Name=dyn_hair_l_02, swing1LimitDegrees=35, swing2LimitDegrees=35, twistLimitDegrees=2}
 {
  bone1Name=dyn_hair_l_MSTR
  bone2Name=head
  swing1LimitDegrees=45
  swing1Motion=Free
  swing2LimitDegrees=45
  swing2Motion=Free
  twistLimitDegrees=45
  twistMotion=Free
 }
 {bone1Name=dyn_hair_l_01, bone2Name=dyn_hair_l_MSTR, swing1LimitDegrees=35, swing2LimitDegrees=35, twistLimitDegrees=2}
 {bone1Name=dyn_hair_l_02, bone2Name=dyn_hair_l_01, swing1LimitDegrees=35, swing2LimitDegrees=35, twistLimitDegrees=2}
 {bone1Name=dyn_hair_bk_05, bone2Name=dyn_hair_bk_04, swing1LimitDegrees=25, swing2LimitDegrees=25, twistLimitDegrees=2}
 {
  bone1Name=dyn_hair_bk_MSTR
  bone2Name=head
  swing1LimitDegrees=45
  swing1Motion=Free
  swing2LimitDegrees=45
  swing2Motion=Free
  twistLimitDegrees=45
  twistMotion=Free
 }
 {bone1Name=dyn_hair_bk_01, bone2Name=dyn_hair_bk_MSTR, swing1LimitDegrees=25, swing2LimitDegrees=25, twistLimitDegrees=2}
 {bone1Name=dyn_hair_bk_02, bone2Name=dyn_hair_bk_01, swing1LimitDegrees=25, swing2LimitDegrees=25, twistLimitDegrees=2}
 {bone1Name=dyn_hair_bk_03, bone2Name=dyn_hair_bk_02, swing1LimitDegrees=25, swing2LimitDegrees=25, twistLimitDegrees=2}
 {bone1Name=dyn_hair_bk_04, bone2Name=dyn_hair_bk_03, swing1LimitDegrees=25, swing2LimitDegrees=25, twistLimitDegrees=2}
 {
  bone1Name=dyn_hair_r_bk_MSTR
  bone2Name=head
  swing1LimitDegrees=45
  swing1Motion=Free
  swing2LimitDegrees=45
  swing2Motion=Free
  twistLimitDegrees=45
  twistMotion=Free
 }
 {bone1Name=dyn_hair_r_bk_01, bone2Name=dyn_hair_r_bk_MSTR, swing1LimitDegrees=25, swing2LimitDegrees=25, twistLimitDegrees=2}
 {bone1Name=dyn_hair_r_bk_02, bone2Name=dyn_hair_r_bk_01, swing1LimitDegrees=25, swing2LimitDegrees=25, twistLimitDegrees=2}
 {bone1Name=dyn_hair_r_bk_03, bone2Name=dyn_hair_r_bk_02, swing1LimitDegrees=25, swing2LimitDegrees=25, twistLimitDegrees=2}
 {bone1Name=dyn_hair_a_r_01, bone2Name=dyn_hair_r_bk_03, swing1LimitDegrees=25, swing2LimitDegrees=25, twistLimitDegrees=2}
 {bone1Name=dyn_hair_b_r_01, bone2Name=dyn_hair_r_bk_03, swing1LimitDegrees=25, swing2LimitDegrees=25, twistLimitDegrees=2}
 {bone1Name=dyn_hair_a_r_02, bone2Name=dyn_hair_a_r_01, swing1LimitDegrees=25, swing2LimitDegrees=25, twistLimitDegrees=2}
 {bone1Name=dyn_hair_b_r_02, bone2Name=dyn_hair_b_r_01, swing1LimitDegrees=25, swing2LimitDegrees=25, twistLimitDegrees=2}
 {
  bone1Name=dyn_hair_r_MSTR
  bone2Name=head
  swing1LimitDegrees=45
  swing1Motion=Free
  swing2LimitDegrees=45
  swing2Motion=Free
  twistLimitDegrees=45
  twistMotion=Free
 }
 {bone1Name=dyn_hair_r_01, bone2Name=dyn_hair_r_MSTR, swing1LimitDegrees=35, swing2LimitDegrees=35, twistLimitDegrees=2}
 {bone1Name=dyn_hair_r_02, bone2Name=dyn_hair_r_01, swing1LimitDegrees=35, swing2LimitDegrees=35, twistLimitDegrees=2}
 {bone1Name=dyn_hair_r_03, bone2Name=dyn_hair_r_02, swing1LimitDegrees=35, swing2LimitDegrees=35, twistLimitDegrees=2}
 {bone1Name=dyn_hair_r_04, bone2Name=dyn_hair_r_03, swing1LimitDegrees=35, swing2LimitDegrees=35, twistLimitDegrees=2}
]
