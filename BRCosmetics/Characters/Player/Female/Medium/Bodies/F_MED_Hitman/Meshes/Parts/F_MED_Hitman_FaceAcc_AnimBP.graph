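# Post-copy animation Blueprint for the female Hitman cosmetic's face accessory part, driving hair motion.
# Copies the parent pose, layers body-mesh arms on it, then feeds a hair Control Rig aware of pistol and rifle targeting.
# Skydiving state alternates two RigidBody sims, neck-based main hair and head-based dynamic hair, before a LOD0-only Control Rig.
# The update graph derives weapon, targeting, skydiving and LOD alphas, acceleration scale and a parameter-asset gravity override.
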AnimBlueprint /BRCosmetics/Characters/Player/Female/Medium/Bodies/F_MED_Hitman/Meshes/Parts/F_MED_Hitman_FaceAcc_AnimBP
target_skeleton=&/F_MED_Hitman_FaceAcc_Skeleton
graphs [
 AnimGraph {nodes [
   n1 = Control Rig!(
    Source=Use cached pose 'PrePose'()
    Alpha
    Is Pistol Targeting:double=Get Is Pistol Targeting Alpha() -> (Is Pistol Targeting Alpha:float)
    Is Rifle Targeting:double=Get Is Rifle Targeting Alpha() -> (Is Rifle Targeting Alpha:float)
   ) -> (Pose) {
    alphaBoolBlend {blendOption=Linear}
    controlRigAssetReference {blueprintRigClass=&/F_MED_Hitman_FaceAcc_Hair_CtrlRig.F_MED_Hitman_FaceAcc_Hair_CtrlRig_C}
   }
   n2 = RigidBody!(
    ComponentPose=Local To Component(LocalPose=@n1)
    Alpha=Get isNotSkydivingFloat() -> (isNotSkydivingFloat:double)
    OverrideWorldGravity:Vector=Get GravityOverride()
    ExternalForce
    ComponentLinearAccScale:Vector=Get overrideAccScale() -> (overrideAccScale:Vector)
    ComponentLinearVelScale:Vector="0.4,0.3,0.3"
   ) -> (Pose) {
    bClampLinearTranslationLimitToRefPose=true
    bOverrideWorldGravity=true
    overridePhysicsAsset=&/F_MED_Hitman_FaceAcc_MainHair_Physics
    simulationSpace=BaseBoneSpace
    alphaBoolBlend {blendOption=Linear}
    alphaScaleBiasClamp {bInterpResult=true, interpSpeedIncreasing=2}
    baseBoneRef {boneName=neck_02}
    componentAppliedLinearAccClamp=<3000 3000 3000>
    componentLinearAccScale=<0.4 0.3 0.3>
    componentLinearVelScale=<0.4 0.3 0.3>
    simSpaceSettings {
     maxAngularAcceleration=1000
     maxAngularVelocity=3
     maxLinearAcceleration=700
     maxLinearVelocity=500
     velocityScaleZ=0
     worldAlpha=0.25
     externalLinearDragV=<0.2 0.2>
    }
   }
   n3 = RigidBody!(
    ComponentPose=@n2
    Alpha=Get isSkydivingFloat() -> (isSkydivingFloat:double)
    OverrideWorldGravity=Get GravityOverride()
    ExternalForce
    ComponentLinearAccScale=Get overrideAccScale()
   ) -> (Pose) {
    bClampLinearTranslationLimitToRefPose=true
    bOverrideWorldGravity=true
    overridePhysicsAsset=&/F_MED_Hitman_FaceAcc_Hair_Dyn
    simulationSpace=BaseBoneSpace
    alphaBoolBlend {blendOption=Linear}
    alphaScaleBiasClamp {bClampResult=true, bInterpResult=true, clampMax=0.8}
    baseBoneRef {boneName=head}
    componentAppliedLinearAccClamp=<800 1000 1000>
    componentLinearAccScale=<0.4 0.3 0.3>
    componentLinearVelScale=<0.1 0.15 0.15>
    overrideWorldGravity {z=-980}
   }
   n4 = Control Rig!(Source=Component To Local(ComponentPose=@n3), Alpha=Get isLOD0() -> (isLOD0:double)) -> (Pose) {
    alphaBoolBlend {blendOption=Linear}
    controlRigAssetReference {blueprintRigClass=&/F_MED_Hitman_FaceAcc_CtrlRig.F_MED_Hitman_FaceAcc_CtrlRig_C}
   }
   Output Pose(Result=@n4) {layerGroup=None}
   n5 = Get body Mesh() -> (body Mesh:Skeletal Mesh Component Object Reference)
   Save cached pose 'PrePose'(Pose=Layered blend per bone(BasePose=Copy Pose From Mesh() {bCopyCurves=true, bUseAttachedParent=true}, BlendPoses_0=Copy Pose From Mesh(SourceMeshComponent=@n5)) {layerSetup [{branchFilters [{boneName=upperarm_l}, {boneName=upperarm_r}]}]}) {cacheName=PrePose}
  ]}
 EventGraph {nodes [
   Event BlueprintUpdateAnimation()
   > n1 = Sequence!(execute) -> (then_0, then_1, then_2:Exec, then_3:Exec, then_4:Exec)
   > Set isCrouchingOrFloating(isCrouchingOrFloating:bool=OR Boolean!(A=Get bIsCrouching(), B=Get bIsSkydiveFloating(), C:bool=Get bIsSkyDiving()) -> (ReturnValue)) -> (Output_Get:bool)
   > Set overrideAccScale(overrideAccScale:Vector=SelectVector(A="0,0.1,0.1", B="0.1,0.15,0.15", bPickA=Get bIsTargeting())) -> (Output_Get:Vector)
   n2 = Branch(execute=@n1.then_1, Condition="Equal (Integer)"(A=Get CurrentLOD(), B=0) {is=title K2Node_CallFunction})
   > Set isLOD0(isLOD0:double=1) -> (Output_Get:double)
   Set isLOD0(execute=@n2.else, isLOD0=0)
   n3 = "Gravity Override State Function (Parameter Asset) (Nativized)"(
    execute=@n1.then_2
    CharacterPartAnimInstance=Self-Reference()
    GameTimeInSeconds=GetGameTimeInSeconds()
    DebugJointName=spine_05
    Parameters=&/F_MED_Hitman_FaceAcc_Parameters
   )
   > Set GravityOverride(GravityOverride=@n3)
   Set body Mesh(execute=@n1.then_3, body Mesh:Skeletal Mesh Component Object Reference=GetOwnerSkeletalMeshComponent()) -> (Output_Get:Skeletal Mesh Component Object Reference)
   n4 = Sequence!(execute=@n1.then_4) -> (then_0, then_1, then_2, then_3, then_4, then_5:Exec)
   > n5 = Cast To FortPlayerAnimInstance(Object=GetAnimInstance(self=GetOwnerSkeletalMeshComponent()))
   > Set isPistol(isPistol=false)
   > Set isRifle(isRifle:bool=false) -> (Output_Get:bool)
   > n6 = Switch on EFortWeaponCoreAnimation!(execute, Selection:EFortWeaponCoreAnimation Enum=Get WeaponCoreAnim(self=@n5)) -> (
    Melee:Exec
    Pistol:Exec
    Shotgun:Exec
    PaperBlueprint:Exec
    Rifle:Exec
    MeleeOneHand:Exec
    MachinePistol:Exec
    RocketLauncher:Exec
    GrenadeLauncher:Exec
    GoingCommando:Exec
    AssaultRifle:Exec
    TacticalShotgun:Exec
    SniperRifle:Exec
    TrapPlacement:Exec
    ShoulderLauncher:Exec
    AbilityDecoTool:Exec
    Crossbow:Exec
    C4:Exec
    RemoteControl:Exec
    DualWield:Exec
    AR_BullPup:Exec
    AR_ForwardGrip:Exec
    MedPackPaddles:Exec
    SMG_P90:Exec
    AR_DrumGun:Exec
    Consumable_Small:Exec
    Consumable_Large:Exec
    Balloon:Exec
    MountedTurret:Exec
    CreativeTool:Exec
    ExplosiveBow:Exec
    AshtonIndigo:Exec
    AshtonChicago:Exec
    MeleeDualWield:Exec
    Unarmed:Exec
   )
   Set isRifle(execute=[@n6.AssaultRifle, @n6.Rifle, @n6.RocketLauncher, @n6.Shotgun, @n6.SniperRifle, @n6.TacticalShotgun], isRifle=true)
   Set isPistol(execute=@n6.Pistol, isPistol=true)
   n7 = Get isRifle() -> (isRifle:bool)
   n8 = Get bIsTargeting()
   Set Is Rifle Alpha(execute=@n4.then_1, Is Rifle Alpha:float="To Float (Boolean)"(InBool=AND Boolean(A=@n7, B=@n8))) -> (Output_Get:float)
   Set Is Pistol Alpha(execute=@n4.then_2, Is Pistol Alpha:float="To Float (Boolean)"(InBool=Get isPistol() -> (isPistol:bool))) -> (Output_Get:float)
   n9 = Get isPistol()
   n10 = Get bIsTargeting()
   Set Is Pistol Targeting Alpha(execute=@n4.then_3, Is Pistol Targeting Alpha:float="To Float (Boolean)"(InBool=AND Boolean(A=@n9, B=@n10))) -> (Output_Get:float)
   n11 = Get bIsTargeting()
   n12 = Get isRifle()
   Set Is Rifle Targeting Alpha(execute=@n4.then_4, Is Rifle Targeting Alpha:float="To Float (Boolean)"(InBool=AND Boolean(A=@n12, B=@n11))) -> (Output_Get:float)
   n13 = Get bIsSkyDiving()
   Set isNotSkydivingFloat(execute=@n4.then_5, isNotSkydivingFloat:double="To Float (Boolean)"(InBool=NOT Boolean(A=@n13))) -> (Output_Get:double)
   > Set isSkydivingFloat(isSkydivingFloat:double="To Float (Boolean)"(InBool=@n13)) -> (Output_Get:double)
   Event BlueprintInitializeAnimation()
   > Set body Mesh(body Mesh=GetOwnerSkeletalMeshComponent())
  ]}
]
variables [
 body Mesh {type=SkeletalMeshComponent}
 GravityOverride {type=Vector}
 Is Pistol Alpha {type=number}
 Is Pistol Targeting Alpha {type=number}
 Is Rifle Alpha {type=number}
 Is Rifle Targeting Alpha {type=number}
 isCrouchingOrFloating {type=boolean}
 isLOD0 {type=number}
 isNotSkydivingFloat {type=number}
 isPistol {type=boolean}
 isRifle {type=boolean}
 isSkydivingFloat {type=number}
 overrideAccScale {type=Vector}
]
